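# Translucent, unlit, two-sided particle material for anime-style flame licks in the BattleSuit contrail effects.
# Color maps an inverted cartoon flame texture through a blue curve atlas row into HDR emissive.
# Opacity uses a panning fire mask, direction scaled by a dynamic "Random flip", with depth fade and a sine U-edge fade bypassed for cylinders.
# World position offset logic layers regular and staccato panned-gradient pulses along normals, plus a "Hide Flames" pull toward object position.
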Material /BRCosmetics/Characters/Player/Male/Medium/Bodies/M_MED_BattleSuit/FX/M_BattleSuit_Contrails_Licks
graph {
 properties {
  bUsedWithNiagaraSprites=true
  bUsedWithParticleSprites=true
  blendMode=BLEND_Translucent
  pixelDepthOffsetMode=PDOM_Legacy
  shadingModel=MSM_Unlit
  twoSided=true
  usedShadingModels=Unlit
 }
 nodes [
  n1 = TextureCoordinate()
  n2 = MaterialFunctionCall(XY=Add(A=TextureCoordinate() {uTiling=2}) {constB=3.66}) {materialFunction=&Swizzle}
  n3 = VectorParameter() {group=WPO_AnimeFlame, parameterName=Regular Pulse Panner Speed, sortPriority=4, defaultValue=<rgba 0 -2 0 0>}
  n4 = ComponentMask(None=@n3.RGB) {g=true, r=true}
  n5 = Multiply(
   A=OneMinus(None=ComponentMask(None=@n1) {g=true})
   B=TextureSample(UVs=Panner(Coordinate=@n2.YX, Speed=@n4) {speedY=10}) {samplerSource=SSM_Wrap_WorldGroupSettings, texture=/Game/Global/Textures/Default/Masks_and_Gradients/gradient_stasis}.R
  )
  n6 = VertexNormalWS()
  n7 = Multiply(
   A=Multiply(A=LinearInterpolate(Alpha=@n5) {constA=-1}, B=@n6)
   B=ScalarParameter() {defaultValue=30, group=WPO_AnimeFlame, parameterName=Regular Pulse WPO Amount}
  )
  n8 = Add(A=TextureCoordinate()) {constB=7}
  n9 = VectorParameter() {group=WPO_AnimeFlame, parameterName=Staccato Pulse Panner Speed, sortPriority=5, defaultValue=<rgba 0 -1 0 0>}
  n10 = TextureSample(UVs=Panner(
    Coordinate=MaterialFunctionCall(XY=@n8) {materialFunction=&Swizzle}.YX
    Speed=ComponentMask(None=@n9.RGB) {g=true, r=true}
   )) {samplerSource=SSM_Wrap_WorldGroupSettings, texture=/Game/Global/Textures/Default/Masks_and_Gradients/T_Vert_Gradient_Sections}
  n11 = Add(
   A=@n7
   B=Multiply(
    A=Multiply(A=@n6, B=@n10.R)
    B=ScalarParameter() {defaultValue=20, group=WPO_AnimeFlame, parameterName=Staccato Pulse WPO Amount, sortPriority=1}
   )
  )
  n12 = ObjectPositionWS()
  n13 = WorldPosition()
  Add(A=@n11, B=Multiply(A=Subtract(A=@n12, B=@n13.XYZ), B=ScalarParameter() {group=WPO, parameterName=Hide Flames}))
  n14 = ScalarParameter() {group=AnimeFlameShape, parameterName=V_Offset}
  n15 = ComponentMask(None=TextureCoordinate()) {r=true}
  n16 = ScalarParameter() {defaultValue=0.5, parameterName=WPO_Push Contrast}
  AppendVector(
   A=Multiply(A=VertexColor().R) {constB=5}
   B=LinearInterpolate(A=@n14, B=ScalarParameter() {defaultValue=0.6, group=AnimeFlameShape, parameterName=V_Offset_High}, Alpha=Power(Base=@n15, Exp=@n16))
  )
  ComponentMask(None=TextureCoordinate()) {g=true}
  n17 = TextureSample() {texture=/Game/Effects/Fort_Effects/Textures/Fire/T_CartoonFlame}
  n18 = CurveAtlasRowParameter(CurveTime=Power(
    Base=OneMinus(None=@n17.G)
    Exp=ScalarParameter() {defaultValue=0.4214, group=Color, parameterName=Curve Contrast Controller, sortPriority=1}
   )) {
   atlas=&AthenaCurveAtlas01
   curve=/Game/Characters/Player/Male/Medium/Bodies/M_MED_BattleSuit/FX/BattleSuit_Blue
   defaultValue=82
   group=Color
   parameterName=Curve Parameter
  }
  n19 = VectorParameter() {group=Color, parameterName=HDR_Tint, sortPriority=2, defaultValue=<rgba 1 1 1 0>}
  MaterialFunctionCall(Emissive=MaterialFunctionCall(
    LDR Input=Clamp(None=@n18)
    HDR Tint=Multiply(A=@n18, B=@n19.RGB)
    HDR Intensity Multiplier=ScalarParameter() {defaultValue=1.5, group=Color, parameterName=HDR_Intensity, sortPriority=3}
   ) {materialFunction=&DeriveHDRfromLDR}.RemappedRange) {materialFunction=&MF_FX_EmissiveAdjustments}
  Multiply(A=@n17.G, B=ParticleColor().A)
  n20 = TextureCoordinate()
  Multiply(A=LinearInterpolate(Alpha=@n10.R) {constA=-1}, B=Sine(None=ComponentMask(None=@n20) {r=true}) {period=2})
  Multiply(A=VertexColor().R) {constB=7}
  n21 = ScalarParameter() {defaultValue=1, group=AnimeFlameShape, parameterName=UScaleCylinder, sortPriority=2}
  n22 = ScalarParameter() {defaultValue=1, group=AnimeFlameShape, parameterName=UScale, sortPriority=1}
  n23 = Multiply(A=TextureCoordinate(), B=AppendVector(A=StaticSwitchParameter(True=@n21, False=@n22) {parameterName=Use on Cylinder}, B=Constant() {r=1}))
  n24 = VectorParameter() {group=AnimeFlameShape, parameterName=Anime Flame Shape Panner Speed, sortPriority=3, defaultValue=<rgba -1 0 0 0>}
  n25 = ComponentMask(None=@n24.RGB) {g=true, r=true}
  n26 = DynamicParameter() {paramNames [Random flip, Param2, Param3, Param4]}
  n27 = TextureSample(UVs=Panner(Coordinate=Add(A=@n23, B=@n14), Speed=Multiply(A=@n25, B=@n26.Random flip)) {bFractionalPart=true}) {texture=/Game/Effects/Fort_Effects/Textures/Fire/T_Fire_Mask}
  n28 = TextureCoordinate()
  n29 = StaticSwitchParameter(True=Constant() {r=1}, False=Sine(None=ComponentMask(None=@n28) {r=true}) {period=2}) {parameterName=Use on Cylinder}
  Multiply(
   A=ScalarParameter() {defaultValue=1, group=Alpha, parameterName=MasterAlpha}
   B=DepthFade(
    Opacity=Multiply(A=@n27.R, B=@n29)
    FadeDistance=ScalarParameter() {defaultValue=10, group=Alpha, parameterName=Fade Distance, sortPriority=1}
   )
  )
  Multiply(A=VertexColor().R) {constB=3.666}
  n30 = TextureCoordinate()
  Multiply(A=OneMinus(None=ComponentMask(None=@n30) {g=true}), B=@n10.R)
  Multiply(A=@n7, B=ComponentMask(None=TextureCoordinate()) {r=true})
  n31 = TextureCoordinate()
  n32 = Power(Base=Add(A=ComponentMask(None=@n31) {g=true}) {constB=0.05}, Exp=ScalarParameter() {defaultValue=2, parameterName=BottomExp})
  Multiply(A=OneMinus(None=Saturate(None=@n32)), B=@n11)
  Multiply()
 ]
}
