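# Physics asset for the male medium BattleSuit body's dynamic flaps, with capsules on pelvis and thighs representing the body.
# Three two-segment back flaps and three one-segment front flaps are boxes chained from small capsule master bodies.
# Back and front-center masters attach to the pelvis, front side masters to the thighs, all with free angular motion.
# Segment joints are limited, with back flaps narrow on one swing axis and wide on the other.
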PhysicsAsset /BRCosmetics/Characters/Player/Male/Medium/Bodies/M_MED_BattleSuit/Meshes/M_MED_Battle_Suit_T1_Dyn
bodies {
 dyn_flap_bk_01 [{shapeType=Box, extentX=7.5, extentY=7.75, extentZ=1.5, center=<3.9911 0 0.0001>}]
 dyn_flap_bk_02 [{shapeType=Box, extentX=7.5, extentY=7.75, extentZ=1.5, center=<3.9911 0 0.0001>}]
 dyn_flap_bk_MSTR [{shapeType=Capsule, length=0.5, radius=0.5, rotation=<rot 90 0 0>}]
 dyn_flap_bk_l_01 [{shapeType=Box, extentX=7.5, extentY=7.75, extentZ=1.5, center=<3.9911 0 0.0001>}]
 dyn_flap_bk_l_02 [{shapeType=Box, extentX=7.5, extentY=7.75, extentZ=1.5, center=<3.9911 0 0.0001>}]
 dyn_flap_bk_l_MSTR [{shapeType=Capsule, length=0.5, radius=0.5, rotation=<rot 90 0 0>}]
 dyn_flap_bk_r_01 [{shapeType=Box, extentX=7.5, extentY=7.75, extentZ=1.5, center=<-3.9911 0 -0.0001>}]
 dyn_flap_bk_r_02 [{shapeType=Box, extentX=7.5, extentY=7.75, extentZ=1.5, center=<-3.9911 0 -0.0001>}]
 dyn_flap_bk_r_MSTR [{shapeType=Capsule, length=0.5, radius=0.5, rotation=<rot 90 0 0>}]
 dyn_flap_fr_02 [{shapeType=Box, extentX=7.5, extentY=7.75, extentZ=1.5, center=<3.9911 0 0.0001>}]
 dyn_flap_fr_MSTR [{shapeType=Capsule, length=0.5, radius=0.5, rotation=<rot 90 0 0>}]
 dyn_flap_fr_l_02 [{shapeType=Box, extentX=7.5, extentY=7.75, extentZ=1.5, center=<3.9911 0 0.0001>}]
 dyn_flap_fr_l_MSTR [{shapeType=Capsule, length=0.5, radius=0.5, rotation=<rot 90 0 0>}]
 dyn_flap_fr_r_02 [{shapeType=Box, extentX=7.5, extentY=7.75, extentZ=1.5, center=<-3.9911 0 0>}]
 dyn_flap_fr_r_MSTR [{shapeType=Capsule, length=0.5, radius=0.5, rotation=<rot 90 0 0>}]
 pelvis [
  {shapeType=Capsule, length=8.844, radius=10.7145, center=<-3.7371 -1.33 0>}
  {shapeType=Capsule, length=3.7188, radius=10.5938, center=<-5.1137 1.1861 -0.0001>, rotation=<rot -90 0 -8.9435>}
 ]
 thigh_l [{shapeType=Capsule, length=16.3438, radius=8.8438, center=<-12.4332 -0.0367 0.0726>, rotation=<rot 90 19.7672 20.3669>}]
 thigh_r [{shapeType=Capsule, length=15.4375, radius=9.75, center=<12.5172 0.3859 0>, rotation=<rot 90 179.1048 -178.1965>}]
}
constraints [
 {bone1Name=dyn_flap_bk_l_01, bone2Name=dyn_flap_bk_l_MSTR, swing1LimitDegrees=10, swing2LimitDegrees=80, twistLimitDegrees=15}
 {bone1Name=dyn_flap_bk_l_02, bone2Name=dyn_flap_bk_l_01, swing1LimitDegrees=10, swing2LimitDegrees=80, twistLimitDegrees=15}
 {bone1Name=dyn_flap_bk_01, bone2Name=dyn_flap_bk_MSTR, swing1LimitDegrees=10, swing2LimitDegrees=80, twistLimitDegrees=15}
 {bone1Name=dyn_flap_bk_02, bone2Name=dyn_flap_bk_01, swing1LimitDegrees=10, swing2LimitDegrees=80, twistLimitDegrees=15}
 {bone1Name=dyn_flap_bk_r_01, bone2Name=dyn_flap_bk_r_MSTR, swing1LimitDegrees=10, swing2LimitDegrees=80, twistLimitDegrees=15}
 {bone1Name=dyn_flap_bk_r_02, bone2Name=dyn_flap_bk_r_01, swing1LimitDegrees=10, swing2LimitDegrees=80, twistLimitDegrees=15}
 {
  bone1Name=dyn_flap_fr_MSTR
  bone2Name=pelvis
  swing1LimitDegrees=45
  swing1Motion=Free
  swing2LimitDegrees=45
  swing2Motion=Free
  twistLimitDegrees=45
  twistMotion=Free
 }
 {
  bone1Name=dyn_flap_bk_l_MSTR
  bone2Name=pelvis
  swing1LimitDegrees=45
  swing1Motion=Free
  swing2LimitDegrees=45
  swing2Motion=Free
  twistLimitDegrees=45
  twistMotion=Free
 }
 {
  bone1Name=dyn_flap_bk_MSTR
  bone2Name=pelvis
  swing1LimitDegrees=45
  swing1Motion=Free
  swing2LimitDegrees=45
  swing2Motion=Free
  twistLimitDegrees=45
  twistMotion=Free
 }
 {
  bone1Name=dyn_flap_bk_r_MSTR
  bone2Name=pelvis
  swing1LimitDegrees=45
  swing1Motion=Free
  swing2LimitDegrees=45
  swing2Motion=Free
  twistLimitDegrees=45
  twistMotion=Free
 }
 {
  bone1Name=thigh_l
  bone2Name=pelvis
  swing1LimitDegrees=45
  swing1Motion=Free
  swing2LimitDegrees=45
  swing2Motion=Free
  twistLimitDegrees=45
  twistMotion=Free
 }
 {
  bone1Name=thigh_r
  bone2Name=pelvis
  swing1LimitDegrees=45
  swing1Motion=Free
  swing2LimitDegrees=45
  swing2Motion=Free
  twistLimitDegrees=45
  twistMotion=Free
 }
 {bone1Name=dyn_flap_fr_02, bone2Name=dyn_flap_fr_MSTR, swing1LimitDegrees=45, swing2LimitDegrees=45, twistLimitDegrees=15}
 {
  bone1Name=dyn_flap_fr_l_MSTR
  bone2Name=thigh_l
  swing1LimitDegrees=45
  swing1Motion=Free
  swing2LimitDegrees=45
  swing2Motion=Free
  twistLimitDegrees=45
  twistMotion=Free
 }
 {bone1Name=dyn_flap_fr_l_02, bone2Name=dyn_flap_fr_l_MSTR, swing1LimitDegrees=45, swing2LimitDegrees=45, twistLimitDegrees=15}
 {
  bone1Name=dyn_flap_fr_r_MSTR
  bone2Name=thigh_r
  swing1LimitDegrees=45
  swing1Motion=Free
  swing2LimitDegrees=45
  swing2Motion=Free
  twistLimitDegrees=45
  twistMotion=Free
 }
 {bone1Name=dyn_flap_fr_r_02, bone2Name=dyn_flap_fr_r_MSTR, swing1LimitDegrees=45, swing2LimitDegrees=45, twistLimitDegrees=15}
]
