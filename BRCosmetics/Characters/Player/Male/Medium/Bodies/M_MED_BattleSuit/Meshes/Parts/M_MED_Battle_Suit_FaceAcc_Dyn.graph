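# Physics asset for the male medium Battle Suit face accessory, covering its dynamic front hair, rear hood and drawstring bones with capsule bodies.
# Three front hair bodies constrain directly to the head, and two hood bodies to spine_05, with angular limits.
# Left and right drawstrings are three-link chains with tight twist limits, rooted on master bodies attached to spine_05.
# Large spine capsules form the torso, with free angular joints linking spine, head and string masters.
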PhysicsAsset /BRCosmetics/Characters/Player/Male/Medium/Bodies/M_MED_BattleSuit/Meshes/Parts/M_MED_Battle_Suit_FaceAcc_Dyn
bodies {
 dyn_hair_fr_01 [{shapeType=Capsule, length=2.1211, radius=1.0309, center=<1 0 0>, rotation=<rot 90 0 0>}]
 dyn_hair_fr_02 [{shapeType=Capsule, length=2.2271, radius=0.9845, center=<1 0 0>, rotation=<rot 90 0 0>}]
 dyn_hair_fr_03 [{shapeType=Capsule, length=2.0877, radius=1.1377, center=<1 0 0>, rotation=<rot 90 0 0>}]
 dyn_hood_bk_l [{shapeType=Capsule, length=2, radius=1, center=<0.4 0 0>, rotation=<rot 90 0 0>}]
 dyn_hood_bk_r [{shapeType=Capsule, length=2, radius=1, center=<0.4 0 0>, rotation=<rot 90 0 0>}]
 dyn_string_l_01 [{shapeType=Capsule, length=5, radius=0.5, center=<3 0 0>, rotation=<rot 90 0 0>}]
 dyn_string_l_02 [{shapeType=Capsule, length=5, radius=0.5, center=<3 0 0>, rotation=<rot 90 0 0>}]
 dyn_string_l_03 [{shapeType=Capsule, length=5, radius=0.5, center=<3 0 0>, rotation=<rot 90 0 0>}]
 dyn_string_l_MSTR [{shapeType=Capsule, length=0.505, radius=0.505, rotation=<rot 0 -90 0>}]
 dyn_string_r_01 [{shapeType=Capsule, length=5, radius=0.5, center=<3 0 0>, rotation=<rot 90 0 0>}]
 dyn_string_r_02 [{shapeType=Capsule, length=5, radius=0.5, center=<3 0 0>, rotation=<rot 90 0 0>}]
 dyn_string_r_03 [{shapeType=Capsule, length=5, radius=0.5, center=<3 0 0>, rotation=<rot 90 0 0>}]
 dyn_string_r_MSTR [{shapeType=Capsule, length=0.505, radius=0.505, rotation=<rot 0 -90 0>}]
 head [{shapeType=Capsule, length=1, radius=1, rotation=<rot 90 0 0>}]
 spine_03 [{shapeType=Capsule, length=7.2575, radius=12.3, shapeName=spine_03_capsule}]
 spine_04 [{shapeType=Capsule, length=7.9062, radius=12.0938, center=<3.1572 0.3 0>}]
 spine_05 [{shapeType=Capsule, length=8.75, radius=13.2188, center=<-4.5203 -0.9652 0>}]
}
constraints [
 {bone1Name=dyn_string_l_01, bone2Name=dyn_string_l_MSTR, swing1LimitDegrees=45, swing2LimitDegrees=45, twistLimitDegrees=2}
 {bone1Name=dyn_string_l_02, bone2Name=dyn_string_l_01, swing1LimitDegrees=45, swing2LimitDegrees=45, twistLimitDegrees=2}
 {bone1Name=dyn_string_l_03, bone2Name=dyn_string_l_02, swing1LimitDegrees=45, swing2LimitDegrees=45, twistLimitDegrees=2}
 {bone1Name=dyn_string_r_01, bone2Name=dyn_string_r_MSTR, swing1LimitDegrees=45, swing2LimitDegrees=45, twistLimitDegrees=2}
 {bone1Name=dyn_string_r_02, bone2Name=dyn_string_r_01, swing1LimitDegrees=45, swing2LimitDegrees=45, twistLimitDegrees=2}
 {bone1Name=dyn_string_r_03, bone2Name=dyn_string_r_02, swing1LimitDegrees=45, swing2LimitDegrees=45, twistLimitDegrees=2}
 {bone1Name=dyn_hair_fr_02, bone2Name=head, swing1LimitDegrees=40, swing2LimitDegrees=20, twistLimitDegrees=2}
 {bone1Name=dyn_hair_fr_01, bone2Name=head, swing1LimitDegrees=40, swing2LimitDegrees=20, twistLimitDegrees=2}
 {bone1Name=dyn_hair_fr_03, bone2Name=head, swing1LimitDegrees=40, swing2LimitDegrees=20, twistLimitDegrees=2}
 {
  bone1Name=head
  bone2Name=spine_05
  swing1LimitDegrees=45
  swing1Motion=Free
  swing2LimitDegrees=45
  swing2Motion=Free
  twistLimitDegrees=45
  twistMotion=Free
 }
 {
  bone1Name=dyn_string_l_MSTR
  bone2Name=spine_05
  swing1LimitDegrees=45
  swing1Motion=Free
  swing2LimitDegrees=45
  swing2Motion=Free
  twistLimitDegrees=45
  twistMotion=Free
 }
 {
  bone1Name=dyn_string_r_MSTR
  bone2Name=spine_05
  swing1LimitDegrees=45
  swing1Motion=Free
  swing2LimitDegrees=45
  swing2Motion=Free
  twistLimitDegrees=45
  twistMotion=Free
 }
 {
  bone1Name=spine_05
  bone2Name=spine_04
  swing1LimitDegrees=45
  swing1Motion=Free
  swing2LimitDegrees=45
  swing2Motion=Free
  twistLimitDegrees=45
  twistMotion=Free
 }
 {bone1Name=dyn_hood_bk_l, bone2Name=spine_05, swing1LimitDegrees=60, swing2LimitDegrees=60, twistLimitDegrees=20}
 {bone1Name=dyn_hood_bk_r, bone2Name=spine_05, swing1LimitDegrees=60, swing2LimitDegrees=60, twistLimitDegrees=20}
 {
  bone1Name=spine_03
  bone2Name=spine_04
  swing1LimitDegrees=45
  swing1Motion=Free
  swing2LimitDegrees=45
  swing2Motion=Free
  twistLimitDegrees=45
  twistMotion=Free
 }
]
