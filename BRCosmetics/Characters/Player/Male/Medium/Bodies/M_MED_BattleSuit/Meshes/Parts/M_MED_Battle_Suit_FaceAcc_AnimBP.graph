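# Anim Blueprint for the male BattleSuit face accessory, animating its two dangling string chains atop the pose copied from the attached parent mesh.
# Bool blends select RigidBody simulation, otherwise AnimDynamics with spine sphere limits, otherwise Trail controllers.
# Additive master-bone offsets follow an eased melee alpha in all paths, plus front-end offsets in the non-RigidBody paths.
# Targeting events halve the RigidBody alpha while aiming and restore it afterwards.
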AnimBlueprint /BRCosmetics/Characters/Player/Male/Medium/Bodies/M_MED_BattleSuit/Meshes/Parts/M_MED_Battle_Suit_FaceAcc_AnimBP
target_skeleton=&/M_MED_Battle_Suit_FaceAcc_Skeleton
graphs [
 AnimGraph {nodes [
   n1 = Trail controller - Bone: dyn_string_l_03(ComponentPose=Local To Component(LocalPose=Copy Pose From Mesh() {bUseAttachedParent=true}), Alpha=1) {
    bLimitRotation=true
    bUsePlanarLimit=true
    chainLength=3
    maxDeltaTime=0.3
    fakeVelocity {z=30}
    planarLimits [{drivingBone {boneName=dyn_string_l_MSTR}, planeTransform {location=<0 0 1>, rotation=<rot -15 0 0>, scale=<1 1 1>}}]
    rotationLimits [{limitMax=<0 20 50>, limitMin=<0 -20 -50>}, {limitMax=<0 20 50>, limitMin=<0 -20 -50>}, {limitMax=<0 20 50>, limitMin=<0 -20 -50>}]
    rotationOffsets [<0 0 0>, <0 0 0>, <0 0 0>]
    trailBone {boneName=dyn_string_l_03}
    trailRelaxationSpeed {editorCurveData {keys [<at 0 5 RCIM_Linear>,<at 1 6 RCIM_Linear>]}}
   }
   n2 = Trail controller - Bone: dyn_string_r_03(ComponentPose=@n1, Alpha=1) {
    bLimitRotation=true
    bUsePlanarLimit=true
    chainLength=3
    maxDeltaTime=0.3
    fakeVelocity {z=30}
    planarLimits [{drivingBone {boneName=dyn_string_r_MSTR}, planeTransform {location=<0 0 1>, rotation=<rot -15 0 0>, scale=<1 1 1>}}]
    rotationLimits [{limitMax=<0 20 50>, limitMin=<0 -20 -50>}, {limitMax=<0 20 50>, limitMin=<0 -20 -50>}, {limitMax=<0 20 50>, limitMin=<0 -20 -50>}]
    rotationOffsets [<0 0 0>, <0 0 0>, <0 0 0>]
    trailBone {boneName=dyn_string_r_03}
    trailRelaxationSpeed {editorCurveData {keys [<at 0 5 RCIM_Linear>,<at 1 6 RCIM_Linear>]}}
   }
   n3 = Get melee started() -> (melee started:double)
   n4 = "Transform (Modify) Bone - Bone: dyn_string_l_MSTR"(ComponentPose=@n2, Alpha=@n3, Translation="0,0,0.5", Rotation="1,0,0", Scale="1,1,1") {
    rotationMode=BMM_Additive
    rotationSpace=BCS_BoneSpace
    translationMode=BMM_Additive
    translationSpace=BCS_BoneSpace
    alphaBoolBlend {blendInTime=0.1, blendOption=ExpInOut, blendOutTime=0.1}
    boneToModify {boneName=dyn_string_l_MSTR}
    rotation {pitch=1}
    translation {z=0.5}
   }
   n5 = "Transform (Modify) Bone - Bone: dyn_string_r_MSTR"(ComponentPose=@n4, Alpha=@n3, Translation="0,0,0.5", Rotation="1,0,0", Scale="1,1,1") {
    rotationMode=BMM_Additive
    rotationSpace=BCS_BoneSpace
    translationMode=BMM_Additive
    translationSpace=BCS_BoneSpace
    alphaBoolBlend {blendInTime=0.1, blendOption=ExpInOut, blendOutTime=0.1}
    boneToModify {boneName=dyn_string_r_MSTR}
    rotation {pitch=1}
    translation {z=0.5}
   }
   n6 = Get bIsInFrontEnd()
   n7 = "Transform (Modify) Bone - Bone: dyn_string_r_MSTR"!(ComponentPose=@n5, bAlphaBoolEnabled:bool=@n6, Translation="0,0,0", Rotation="12,0,0", Scale="1,1,1") -> (Pose) {
    alphaInputType=Bool
    rotationMode=BMM_Additive
    rotationSpace=BCS_BoneSpace
    translationMode=BMM_Additive
    translationSpace=BCS_BoneSpace
    alphaBoolBlend {blendInTime=0.1, blendOption=ExpInOut, blendOutTime=0.1}
    boneToModify {boneName=dyn_string_r_MSTR}
    rotation {pitch=12}
   }
   n8 = "Transform (Modify) Bone - Bone: dyn_string_l_MSTR"!(ComponentPose=@n7, bAlphaBoolEnabled:bool=@n6, Translation="0,0,0", Rotation="10,0,0", Scale="1,1,1") -> (Pose) {
    alphaInputType=Bool
    rotationMode=BMM_Additive
    rotationSpace=BCS_BoneSpace
    translationMode=BMM_Additive
    translationSpace=BCS_BoneSpace
    alphaBoolBlend {blendInTime=0.1, blendOption=ExpInOut, blendOutTime=0.1}
    boneToModify {boneName=dyn_string_l_MSTR}
    rotation {pitch=10}
   }
   n9 = Component To Local(ComponentPose=@n8)
   n10 = "Anim Dynamics - Chain: dyn_string_l_01 -> dyn_string_l_03"!(
    ComponentPose=Local To Component(LocalPose=Copy Pose From Mesh() {bUseAttachedParent=true})
    Alpha=1
    GravityOverride:Vector="0,0,-980"
    ExternalForce="0,0,0"
   ) -> (Pose) {
    angularDampingOverride=0.96
    angularSpringConstant=150
    bAngularSpring=true
    bChain=true
    bOverrideAngularDamping=true
    bUseGravityOverride=true
    bUsePlanarLimit=false
    bUseSphericalLimits=true
    alphaBoolBlend {blendOption=Linear}
    boundBone {boneName=dyn_string_l_01}
    chainEnd {boneName=dyn_string_l_03}
    componentAppliedLinearAccClamp=<50 50 50>
    componentLinearAccScale=<0.1 0.1 0.1>
    componentLinearVelScale=<0.1 0.1 0.1>
    physicsBodyDefinitions [
     {
      boundBone {boneName=dyn_string_l_01}
      boxExtents=<50 50 50>
      constraintSetup {angularLimitsMax=<5 5 5>, angularLimitsMin=<-5 -5 -5>, angularTarget=<0 0 0>}
      localJointOffset=<0 0 0>
     }
     {
      boundBone {boneName=dyn_string_l_02}
      boxExtents=<20 20 20>
      constraintSetup {angularLimitsMax=<10 10 10>, angularLimitsMin=<-10 -10 -10>, angularTarget=<0 0 0>}
      localJointOffset=<0 0 0>
     }
     {
      boundBone {boneName=dyn_string_l_03}
      boxExtents=<12 12 12>
      constraintSetup {angularLimitsMax=<10 10 10>, angularLimitsMin=<-10 -10 -10>, angularTarget=<0 0 0>}
      localJointOffset=<0 0 0>
     }
    ]
    sphericalLimits [
     {limitRadius=13.5, drivingBone {boneName=spine_04}, sphereLocalOffset=<0 0 0>}
     {limitRadius=12, drivingBone {boneName=spine_05}, sphereLocalOffset=<0 0 0>}
    ]
   }
   n11 = "Anim Dynamics - Chain: dyn_string_r_01 -> dyn_string_r_03"!(ComponentPose=@n10, Alpha=1, GravityOverride:Vector="0,0,-980", ExternalForce="0,0,0") -> (Pose) {
    angularDampingOverride=0.97
    angularSpringConstant=160
    bAngularSpring=true
    bChain=true
    bOverrideAngularDamping=true
    bUseGravityOverride=true
    bUsePlanarLimit=false
    bUseSphericalLimits=true
    alphaBoolBlend {blendOption=Linear}
    boundBone {boneName=dyn_string_r_01}
    chainEnd {boneName=dyn_string_r_03}
    componentAppliedLinearAccClamp=<48 49 50>
    componentLinearAccScale=<0.1 0.1 0.1>
    componentLinearVelScale=<0.1 0.1 0.1>
    physicsBodyDefinitions [
     {
      boundBone {boneName=dyn_string_r_01}
      boxExtents=<52 48 50>
      constraintSetup {angularLimitsMax=<5 5 5>, angularLimitsMin=<-5 -5 -5>, angularTarget=<0 0 0>}
      localJointOffset=<0 0 0>
     }
     {
      boundBone {boneName=dyn_string_r_02}
      boxExtents=<22 20 18>
      constraintSetup {angularLimitsMax=<10 10 10>, angularLimitsMin=<-10 -10 -10>, angularTarget=<0 0 0>}
      localJointOffset=<0 0 0>
     }
     {
      boundBone {boneName=dyn_string_r_03}
      boxExtents=<13 11 14>
      constraintSetup {angularLimitsMax=<10 10 10>, angularLimitsMin=<-10 -10 -10>, angularTarget=<0 0 0>}
      localJointOffset=<0 0 0>
     }
    ]
    sphericalLimits [
     {limitRadius=13.5, drivingBone {boneName=spine_04}, sphereLocalOffset=<0 0 0>}
     {limitRadius=12, drivingBone {boneName=spine_05}, sphereLocalOffset=<0 0 0>}
    ]
   }
   n12 = Get melee started()
   n13 = "Transform (Modify) Bone - Bone: dyn_string_l_MSTR"!(ComponentPose=@n11, Alpha:float=@n12, Translation="0,0,1", Rotation="2,0,0", Scale="1,1,1") -> (Pose) {
    rotationMode=BMM_Additive
    rotationSpace=BCS_BoneSpace
    translationMode=BMM_Additive
    translationSpace=BCS_BoneSpace
    alphaBoolBlend {blendInTime=0.1, blendOption=ExpInOut, blendOutTime=0.1}
    boneToModify {boneName=dyn_string_l_MSTR}
    rotation {pitch=2}
    translation {z=1}
   }
   n14 = "Transform (Modify) Bone - Bone: dyn_string_r_MSTR"!(ComponentPose=@n13, Alpha:float=@n12, Translation="0,0,1", Rotation="2,0,0", Scale="1,1,1") -> (Pose) {
    rotationMode=BMM_Additive
    rotationSpace=BCS_BoneSpace
    translationMode=BMM_Additive
    translationSpace=BCS_BoneSpace
    alphaBoolBlend {blendInTime=0.1, blendOption=ExpInOut, blendOutTime=0.1}
    boneToModify {boneName=dyn_string_r_MSTR}
    rotation {pitch=2}
    translation {z=1}
   }
   n15 = "Transform (Modify) Bone - Bone: dyn_string_r_MSTR"!(ComponentPose=@n14, bAlphaBoolEnabled:bool=Get bIsInFrontEnd(), Translation="0,0,1", Rotation="0,0,0", Scale="1,1,1") -> (Pose) {
    alphaInputType=Bool
    rotationMode=BMM_Additive
    rotationSpace=BCS_BoneSpace
    translationMode=BMM_Additive
    translationSpace=BCS_BoneSpace
    alphaBoolBlend {blendInTime=0.1, blendOption=ExpInOut, blendOutTime=0.1}
    boneToModify {boneName=dyn_string_r_MSTR}
    translation {z=1}
   }
   n16 = Component To Local(ComponentPose=@n15)
   n17 = Get bAnimDynamicsEnabled() -> (bAnimDynamicsEnabled:bool)
   n18 = RigidBody(
    ComponentPose=Local To Component(LocalPose=Copy Pose From Mesh() {bUseAttachedParent=true})
    Alpha=Get targeting alpha() -> (targeting alpha:double)
   ) {
    bOverrideWorldGravity=true
    overridePhysicsAsset=&/M_MED_Battle_Suit_FaceAcc_Dyn
    simulationSpace=BaseBoneSpace
    alphaBoolBlend {blendOption=Linear}
    baseBoneRef {boneName=spine_03}
    componentAppliedLinearAccClamp=<3000 3000 3000>
    componentLinearAccScale=<0.3 0.3 0.3>
    componentLinearVelScale=<0.3 0.3 0.3>
    overrideWorldGravity {z=-980}
   }
   n19 = Get melee started()
   n20 = "Transform (Modify) Bone - Bone: dyn_string_l_MSTR"(ComponentPose=@n18, Alpha=@n19, Translation="0,0,0.5", Rotation="2,0,0", Scale="1,1,1") {
    rotationMode=BMM_Additive
    rotationSpace=BCS_BoneSpace
    translationMode=BMM_Additive
    translationSpace=BCS_BoneSpace
    alphaBoolBlend {blendInTime=0.1, blendOption=ExpInOut, blendOutTime=0.1}
    boneToModify {boneName=dyn_string_l_MSTR}
    rotation {pitch=2}
    translation {z=0.5}
   }
   n21 = "Transform (Modify) Bone - Bone: dyn_string_r_MSTR"!(ComponentPose=@n20, Alpha:float=@n19, Translation="0,0,0.5", Rotation="2,0,0", Scale="1,1,1") -> (Pose) {
    rotationMode=BMM_Additive
    rotationSpace=BCS_BoneSpace
    translationMode=BMM_Additive
    translationSpace=BCS_BoneSpace
    alphaBoolBlend {blendInTime=0.1, blendOption=ExpInOut, blendOutTime=0.1}
    boneToModify {boneName=dyn_string_r_MSTR}
    rotation {pitch=2}
    translation {z=0.5}
   }
   Output Pose(Result=Blend Poses by bool(
     BlendPose_0=Component To Local(ComponentPose=@n21)
     BlendPose_1=Blend Poses by bool(BlendPose_0=@n16, BlendPose_1=@n9, bActiveValue=@n17) {blendType=Linear}
     bActiveValue=Get bRigidBodyEnabled()
    ) {blendType=Linear}) {layerGroup=None}
  ]}
 EventGraph {nodes [
   n1 = Event BlueprintUpdateAnimation()
   > Set delta time x(delta time x:float=@n1.DeltaTimeX) -> (Output_Get:float)
   > Sequence!(execute) -> (then_0, then_1, then_2:Exec)
   > n2 = Branch(Condition=Get bIsPlayingMeleeAnim())
   > Set melee started(melee started:double=FInterpTo(Current=Get melee started(), Target=1, DeltaTime=Get delta time x() -> (delta time x:float), InterpSpeed=5)) -> (Output_Get:double)
   Set melee started(execute=@n2.else, melee started=FInterpTo(Current=Get melee started(), Target=0, DeltaTime=Get delta time x(), InterpSpeed=2))
   Event OnBeginTargeting() {is=title K2Node_Event}
   > Set targeting alpha(targeting alpha:double=0.5) -> (Output_Get:double)
   Event OnEndTargeting() {is=title K2Node_Event}
   > Set targeting alpha(targeting alpha=1)
   TryGetPawnOwner()
   SelectFloat(A=1, B=0, bPickA=false)
  ]}
]
variables [
 delta time x {type=number}
 melee started {type=number}
 targeting alpha {type=number, default=1}
]
